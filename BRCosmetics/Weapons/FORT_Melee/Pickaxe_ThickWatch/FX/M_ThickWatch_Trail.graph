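# Translucent unlit Niagara ribbon material for the trail effect of the ThickWatch pickaxe.
# Panning soft noise distorts the UV1 lookup of a color texture, which is saturation-adjusted, tinted by particle color and passed through an emissive adjustment function.
# Opacity combines a mask texture, a tiling-method clamp and particle alpha, with adjustable leading and trailing edge fades driven by UV0's horizontal coordinate.
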Material /BRCosmetics/Weapons/FORT_Melee/Pickaxe_ThickWatch/FX/M_ThickWatch_Trail
graph {
 properties {bUsedWithNiagaraRibbons=true, blendMode=BLEND_Translucent, pixelDepthOffsetMode=PDOM_Legacy, shadingModel=MSM_Unlit, usedShadingModels=Unlit}
 nodes [
  n1 = TextureCoordinate() {coordinateIndex=1}
  n2 = MaterialFunctionCall(X=ScalarParameter() {defaultValue=0.5, parameterName=UVScaleX}, Y=ScalarParameter() {defaultValue=2, parameterName=UVScaleY}) {materialFunction=&MakeFloat2}
  n3 = TextureSample(UVs=Panner(Coordinate=Multiply(A=@n1, B=@n2.Result)) {speedX=0.5}) {texture=/Game/Characters/Player/Male/Medium/Bodies/M_MED_Viper/Textures/Glitch/T_SoftNoise}
  n4 = MaterialFunctionCall(X=@n3.R, Y=@n3.R) {materialFunction=&MakeFloat2}
  n5 = ScalarParameter() {defaultValue=-0.1, parameterName=UVDistortAmount}
  n6 = TextureSample(UVs=Add(A=TextureCoordinate() {coordinateIndex=1}, B=Multiply(A=@n4.Result, B=@n5))) {texture=&/T_FX_ThickWatch_TrailMask_Color}
  n7 = ScalarParameter() {defaultValue=-0.5, parameterName=Desaturation}
  n8 = ParticleColor()
  MaterialFunctionCall(Emissive=Multiply(A=Desaturation(None=@n6.RGB, Fraction=@n7), B=@n8.RGB)) {materialFunction=&MF_FX_EmissiveAdjustments}
  n9 = TextureCoordinate() {coordinateIndex=1}
  n10 = VectorParameter() {parameterName=TilingMethod, defaultValue=<rgba 1 0 0 0>}
  n11 = Multiply(
   A=TextureSample(UVs=@n9) {texture=&/T_FX_ThickWatch_TrailMask}.R
   B=MaterialFunctionCall!(
    TexCoord=Panner(Coordinate=@n9)
    ClampChannel=MaterialFunctionCall(X=@n10.R, Y=@n10.R) {materialFunction=&MakeFloat2}.Result
   ) {materialFunction=/Game/Effects/Fort_Effects/MaterialFunctions/MF_FX_TilingMethod}.Result
  )
  n12 = ComponentMask(None=TextureCoordinate()) {r=true}
  n13 = Multiply(
   A=Add(A=@n12, B=ScalarParameter() {group="Fade Leading && Trailing Edges", parameterName=Leading Edge Fade Bias, sortPriority=2})
   B=ScalarParameter() {defaultValue=4, group="Fade Leading && Trailing Edges", parameterName=Leading Edge Fade Scale}
  ) {constB=3}
  n14 = Multiply(
   A=Add(A=OneMinus(None=@n12), B=ScalarParameter() {group="Fade Leading && Trailing Edges", parameterName=Trailing Edge Fade Bias, sortPriority=3})
   B=ScalarParameter() {defaultValue=1, group="Fade Leading && Trailing Edges", parameterName=Trailing Edge Fade Scale, sortPriority=1}
  ) {constB=2}
  n15 = Multiply(
   A=LinearInterpolate(B=Saturate(None=@n13), Alpha=ScalarParameter() {defaultValue=1, group=Alpha, parameterName=Fade Leading Edge, sortPriority=4}) {constA=1}
   B=LinearInterpolate(B=Saturate(None=@n14), Alpha=ScalarParameter() {defaultValue=1, group=Alpha, parameterName=Fade Trailing Edge, sortPriority=5}) {constA=1}
  )
  Multiply(A=@n8.A, B=Saturate(None=Multiply(A=@n11, B=@n15)))
 ]
}
