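# Physics asset for the ThickWatch pickaxe mesh, a cosmetic two-handed melee axe.
# It holds a single capsule body on the Melee_TwoHanded_Handle bone, sized along the handle and offset from the bone origin.
# It has no constraints, so the weapon simulates and collides as one rigid piece.
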PhysicsAsset /BRCosmetics/Weapons/FORT_Melee/Pickaxe_ThickWatch/Meshes/ThickWatch_Axe_PhysicsAsset
bodies {Melee_TwoHanded_Handle [{shapeType=Capsule, length=46.5567, radius=5.0243, center=<0.5806 -0.2025 16.9098>, rotation=<rot -1.1278 -90 0.0081>}]}
constraints []
